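# Sound cue for the Void activation of the VividRazor Baron, routed to the Fort_Gadgets_Items sound class with its own attenuation.
# A mixer layers a main activation wave, delayed a quarter second and widely pitch-varied, over an immediate random pick from four zap variations with narrower pitch variation.
# The random zap node opts out of branch culling, and a concurrency override caps simultaneous instances at a few voices.
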SoundCue /VividRazor_Baron/Audio/SC_VividRazor_Baron_Activate_Void
root=Mixer_0
nodes [
 {id=Delay_0, children [Modulator_0], properties {delayMax=0.25, delayMin=0.25}}
 {id=Mixer_0, children [Delay_0,Modulator_1], properties {inputVolume [1,1]}}
 {id=Modulator_0, children [WavePlayer_1], properties {pitchMax=1.6, pitchMin=0.85}}
 {id=Modulator_1, children [Random_0], properties {pitchMax=1.2, pitchMin=0.85}}
 {
  id=Random_0
  children [WavePlayer_0,WavePlayer_2,WavePlayer_3,WavePlayer_4]
  properties {bShouldExcludeFromBranchCulling=true, weights [1,1,1,1]}
 }
 {id=WavePlayer_0, properties {soundWaveAssetPtr=&/VividRazor_Baron_Activate_Void_Zap_01}}
 {id=WavePlayer_1, properties {soundWaveAssetPtr=/VividRazor_Baron/Audio/wav/VividRazor_Baron_Activate_Void}}
 {id=WavePlayer_2, properties {soundWaveAssetPtr=&/VividRazor_Baron_Activate_Void_Zap_02}}
 {id=WavePlayer_3, properties {soundWaveAssetPtr=&/VividRazor_Baron_Activate_Void_Zap_03}}
 {id=WavePlayer_4, properties {soundWaveAssetPtr=&/VividRazor_Baron_Activate_Void_Zap_04}}
]
properties {
 attenuationSettings=/VividRazor_Baron/Audio/Settings/ATT_VividRazor_Baron_Activate_Void
 bOverrideConcurrency=true
 maxDistance=13000
 soundClassObject=&Fort_Gadgets_Items
 volumeMultiplier=1.3
 concurrencyOverrides {maxCount=4}
}
