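# SC_VividRazor_Baron_Foley is a Sound Cue that plays foley audio for the VividRazor Baron, routed through a melee weapon sound class.
# A Random node picks one of three equally weighted foley waves and feeds a Modulator at the root for playback variation.
# The cue plays at reduced volume with close-range stereo melee attack attenuation and a concurrency override that caps simultaneous instances.
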SoundCue /VividRazor_Baron/Audio/SC_VividRazor_Baron_Foley
root=Modulator_1
nodes [
 {id=Modulator_1, children [Random_1]}
 {id=Random_1, children [WavePlayer_2,WavePlayer_7,WavePlayer_8], properties {weights [1,1,1]}}
 {id=WavePlayer_2, properties {soundWaveAssetPtr=/VividRazor_Baron/Audio/wav/VividRazor_Baron_Foley_01}}
 {id=WavePlayer_7, properties {soundWaveAssetPtr=/VividRazor_Baron/Audio/wav/VividRazor_Baron_Foley_02}}
 {id=WavePlayer_8, properties {soundWaveAssetPtr=/VividRazor_Baron/Audio/wav/VividRazor_Baron_Foley_03}}
]
properties {
 attenuationSettings=/VividRazor_Baron/Audio/Settings/CS_VividRazor_Baron_Melee_Attack_Close_Range_Stereo_Att
 bOverrideConcurrency=true
 maxDistance=10000
 soundClassObject=&Fort_WeaponMelee_PSM
 volumeMultiplier=0.5
 concurrencyOverrides {maxCount=3}
}
