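# Sound cue that plays the howling ping sweetener for VividRazor Baron, assigned to a melee weapon sound class.
# A single wave player feeds a Modulator node left at default settings.
# Playback uses the Baron ping attenuation asset, a boosted volume multiplier and a concurrency override that caps simultaneous instances.
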SoundCue /VividRazor_Baron/Audio/SC_VividRazor_Baron_Howling_Ping
root=Modulator_0
nodes [
 {id=Modulator_0, children [WavePlayer_0]}
 {id=WavePlayer_0, properties {soundWaveAssetPtr=/VividRazor_Baron/Audio/wav/VividRazor_Baron_Howling_PingSweetener_5_1}}
]
properties {
 attenuationSettings=/VividRazor_Baron/Audio/Settings/CS_VividRazor_Baron_Ping_Att
 bOverrideConcurrency=true
 maxDistance=8000
 soundClassObject=&Fort_WeaponMelee_PSM
 volumeMultiplier=2
 concurrencyOverrides {maxCount=3}
}
