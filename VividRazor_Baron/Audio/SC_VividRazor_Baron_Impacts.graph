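# Melee impact sound cue for the VividRazor Baron weapon, capped at three concurrent instances.
# An IsLocal branch sends local or unset playback through stereo attenuation, mixing random first-person impacts with random hit-marker body layers.
# Non-local playback uses random third-person impacts through mono attenuation.
# Quality-level nodes pick six impact variants for the first two levels and a three-variant subset, exempt from branch culling, for the last two.
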SoundCue /VividRazor_Baron/Audio/SC_VividRazor_Baron_Impacts
root=Branch_0
nodes [
 {
  id=Attenuation_0
  children [Mixer_0]
  properties {attenuationSettings=/VividRazor_Baron/Audio/Settings/CS_VividRazor_Baron_Melee_Impact_Stereo_Att}
 }
 {
  id=Attenuation_1
  children [QualityLevel_1]
  properties {attenuationSettings=/VividRazor_Baron/Audio/Settings/CS_VividRazor_Baron_Melee_Impact_Mono_Att}
 }
 {id=Branch_0, children [Modulator_0,Modulator_1,Modulator_0], properties {boolParameterName=IsLocal}}
 {id=Mixer_0, children [QualityLevel_0, Random_4], properties {inputVolume [2.5,1.5]}}
 {id=Modulator_0, children [Attenuation_0]}
 {id=Modulator_1, children [Attenuation_1]}
 {id=QualityLevel_0, children [Random_0,Random_0,Random_2,Random_2]}
 {id=QualityLevel_1, children [Random_1,Random_1,Random_3,Random_3]}
 {
  id=Random_0
  children [WavePlayer_3, WavePlayer_4, WavePlayer_5, WavePlayer_6, WavePlayer_11, WavePlayer_12]
  properties {weights [1,1,1,1,1,1]}
 }
 {
  id=Random_1
  children [WavePlayer_7, WavePlayer_8, WavePlayer_9, WavePlayer_10, WavePlayer_13, WavePlayer_14]
  properties {weights [1,1,1,1,1,1]}
 }
 {id=Random_2, children [WavePlayer_6, WavePlayer_11, WavePlayer_12], properties {bShouldExcludeFromBranchCulling=true, weights [1,1,1]}}
 {id=Random_3, children [WavePlayer_10, WavePlayer_13, WavePlayer_14], properties {bShouldExcludeFromBranchCulling=true, weights [1,1,1]}}
 {id=Random_4, children [WavePlayer_15, WavePlayer_16, WavePlayer_17], properties {weights [1,1,1]}}
 {id=WavePlayer_10, properties {soundWaveAssetPtr=/VividRazor_Baron/Audio/wav/VividRazor_Baron_Impact_3P_04}}
 {id=WavePlayer_11, properties {soundWaveAssetPtr=/VividRazor_Baron/Audio/wav/VividRazor_Baron_Impact_1P_05}}
 {id=WavePlayer_12, properties {soundWaveAssetPtr=/VividRazor_Baron/Audio/wav/VividRazor_Baron_Impact_1P_06}}
 {id=WavePlayer_13, properties {soundWaveAssetPtr=/VividRazor_Baron/Audio/wav/VividRazor_Baron_Impact_3P_05}}
 {id=WavePlayer_14, properties {soundWaveAssetPtr=/VividRazor_Baron/Audio/wav/VividRazor_Baron_Impact_3P_06}}
 {id=WavePlayer_15, properties {soundWaveAssetPtr=/VividRazor_Baron/Audio/wav/VividRazor_Baron_HitMarker_Body_01}}
 {id=WavePlayer_16, properties {soundWaveAssetPtr=/VividRazor_Baron/Audio/wav/VividRazor_Baron_HitMarker_Body_02}}
 {id=WavePlayer_17, properties {soundWaveAssetPtr=/VividRazor_Baron/Audio/wav/VividRazor_Baron_HitMarker_Body_03}}
 {id=WavePlayer_3, properties {soundWaveAssetPtr=/VividRazor_Baron/Audio/wav/VividRazor_Baron_Impact_1P_01}}
 {id=WavePlayer_4, properties {soundWaveAssetPtr=/VividRazor_Baron/Audio/wav/VividRazor_Baron_Impact_1P_02}}
 {id=WavePlayer_5, properties {soundWaveAssetPtr=/VividRazor_Baron/Audio/wav/VividRazor_Baron_Impact_1P_03}}
 {id=WavePlayer_6, properties {soundWaveAssetPtr=/VividRazor_Baron/Audio/wav/VividRazor_Baron_Impact_1P_04}}
 {id=WavePlayer_7, properties {soundWaveAssetPtr=/VividRazor_Baron/Audio/wav/VividRazor_Baron_Impact_3P_01}}
 {id=WavePlayer_8, properties {soundWaveAssetPtr=/VividRazor_Baron/Audio/wav/VividRazor_Baron_Impact_3P_02}}
 {id=WavePlayer_9, properties {soundWaveAssetPtr=/VividRazor_Baron/Audio/wav/VividRazor_Baron_Impact_3P_03}}
]
properties {
 bOverrideConcurrency=true
 maxDistance=3512
 soundClassObject=&Fort_WeaponMelee_PSM
 volumeMultiplier=1
 concurrencyOverrides {maxCount=3}
}
