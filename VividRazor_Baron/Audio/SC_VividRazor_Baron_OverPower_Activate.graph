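# Sound cue for the VividRazor Baron OverPower activation, playing the OverPoweredStart wave.
# A single wave player feeds a Modulator node at the root, with no randomization among waves.
# It uses a dedicated stereo OverPowered attenuation asset, the melee weapon sound class, a raised volume and an overridden concurrency limit.
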SoundCue /VividRazor_Baron/Audio/SC_VividRazor_Baron_OverPower_Activate
root=Modulator_0
nodes [
 {id=Modulator_0, children [WavePlayer_1]}
 {id=WavePlayer_1, properties {soundWaveAssetPtr=/VividRazor_Baron/Audio/wav/VividRazor_Baron_OverPoweredStart_01}}
]
properties {
 attenuationSettings=/VividRazor_Baron/Audio/Settings/CS_VividRazor_Baron_OverPowered_Stereo_Att
 bOverrideConcurrency=true
 maxDistance=10000
 soundClassObject=&Fort_WeaponMelee_PSM
 volumeMultiplier=2
 concurrencyOverrides {maxCount=3}
}
