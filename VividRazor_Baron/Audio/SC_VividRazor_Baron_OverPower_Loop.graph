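# Sound cue for the looping OverPower sound of the VividRazor Baron, routed to the melee weapon sound class.
# A single looping wave player feeds a Modulator left at its default settings, played back at reduced cue volume.
# It uses a dedicated stereo attenuation asset and overrides concurrency to cap simultaneous instances at a small count.
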SoundCue /VividRazor_Baron/Audio/SC_VividRazor_Baron_OverPower_Loop
root=Modulator_0
nodes [
 {id=Modulator_0, children [WavePlayer_0]}
 {id=WavePlayer_0, properties {bLooping=true, soundWaveAssetPtr=/VividRazor_Baron/Audio/wav/VividRazor_Baron_OverPowered_Loop}}
]
properties {
 attenuationSettings=/VividRazor_Baron/Audio/Settings/CS_VividRazor_Baron_OverPowered_Loop_Stereo_Att
 bOverrideConcurrency=true
 maxDistance=10000
 soundClassObject=&Fort_WeaponMelee_PSM
 volumeMultiplier=0.3
 concurrencyOverrides {maxCount=3}
}
