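# Sound cue for the first swipe swing of the VividRazor Baron melee weapon.
# It picks one of three Swing1 wave variations at equal weight, then passes it through a modulator.
# It uses the close-range stereo melee attack attenuation, the Fortnite melee weapon sound class, and a concurrency override that caps simultaneous instances.
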SoundCue /VividRazor_Baron/Audio/SC_VividRazor_Baron_Swipe_01
root=Modulator_0
nodes [
 {id=Modulator_0, children [Random_0]}
 {id=Random_0, children [WavePlayer_0,WavePlayer_1,WavePlayer_2], properties {weights [1,1,1]}}
 {id=WavePlayer_0, properties {soundWaveAssetPtr=/VividRazor_Baron/Audio/wav/VividRazor_Baron_Swing1_01}}
 {id=WavePlayer_1, properties {soundWaveAssetPtr=/VividRazor_Baron/Audio/wav/VividRazor_Baron_Swing1_02}}
 {id=WavePlayer_2, properties {soundWaveAssetPtr=/VividRazor_Baron/Audio/wav/VividRazor_Baron_Swing1_03}}
]
properties {
 attenuationSettings=/VividRazor_Baron/Audio/Settings/CS_VividRazor_Baron_Melee_Attack_Close_Range_Stereo_Att
 bOverrideConcurrency=true
 maxDistance=10000
 soundClassObject=&Fort_WeaponMelee_PSM
 volumeMultiplier=1
 concurrencyOverrides {maxCount=3}
}
