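# Sound cue for the third swipe of the VividRazor Baron melee attack, layering a swing sound with a foley sound.
# A mixer plays both layers at once, each picking randomly among three equally weighted waves through a modulator that slightly lowers pitch and boosts volume at random.
# It uses close-range stereo melee attenuation, the melee weapon sound class and a concurrency override that caps simultaneous instances.
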SoundCue /VividRazor_Baron/Audio/SC_VividRazor_Baron_Swipe_03
root=Mixer_0
nodes [
 {id=Mixer_0, children [Modulator_0,Modulator_1], properties {inputVolume [1,1]}}
 {id=Modulator_0, children [Random_0], properties {pitchMax=1, pitchMin=0.8, volumeMax=1.3, volumeMin=1.2}}
 {id=Modulator_1, children [Random_1], properties {pitchMax=1, pitchMin=0.8, volumeMax=1.3, volumeMin=1.2}}
 {id=Random_0, children [WavePlayer_0,WavePlayer_1,WavePlayer_2], properties {weights [1,1,1]}}
 {id=Random_1, children [WavePlayer_3,WavePlayer_4,WavePlayer_5], properties {weights [1,1,1]}}
 {id=WavePlayer_0, properties {soundWaveAssetPtr=/VividRazor_Baron/Audio/wav/VividRazor_Baron_Swing3_01}}
 {id=WavePlayer_1, properties {soundWaveAssetPtr=/VividRazor_Baron/Audio/wav/VividRazor_Baron_Swing3_02}}
 {id=WavePlayer_2, properties {soundWaveAssetPtr=/VividRazor_Baron/Audio/wav/VividRazor_Baron_Swing3_03}}
 {id=WavePlayer_3, properties {soundWaveAssetPtr=/VividRazor_Baron/Audio/wav/VividRazor_Baron_Swing3_Foley_01}}
 {id=WavePlayer_4, properties {soundWaveAssetPtr=/VividRazor_Baron/Audio/wav/VividRazor_Baron_Swing3_Foley_02}}
 {id=WavePlayer_5, properties {soundWaveAssetPtr=/VividRazor_Baron/Audio/wav/VividRazor_Baron_Swing3_Foley_03}}
]
properties {
 attenuationSettings=/VividRazor_Baron/Audio/Settings/CS_VividRazor_Baron_Melee_Attack_Close_Range_Stereo_Att
 bOverrideConcurrency=true
 maxDistance=10000
 soundClassObject=&Fort_WeaponMelee_PSM
 volumeMultiplier=1.3
 concurrencyOverrides {maxCount=3}
}
